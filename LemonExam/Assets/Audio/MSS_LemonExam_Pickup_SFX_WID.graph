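# One-shot stereo MetaSound source for a LemonExam pickup sound effect, using weapon foley sound class, attenuation and concurrency.
# OnPlay triggers a deluxe stereo wave player, which feeds the left/right outputs and OnFinished.
# The player reads the Wave Asset array, which defaults to empty; Sounds holds the pickup wave but is unconnected.
# Volume, random pitch and character inputs are also unconnected in this graph.
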MetaSoundSource /LemonExam/Assets/Audio/MSS_LemonExam_Pickup_SFX_WID
inputs [
 Character.Affiliation {type=Enum:FortAudioAffiliation}
 Character.Health {type=Float}
 Character.UniqueID {type=Int32}
 MasterVolume {type=Float}
 Random.PitchShift.Base {type=Float}
 Random.PitchShift.Spread {type=Float}
 Sounds {type=WaveAsset:Array}
 UE.Source.OnPlay {type=Trigger}
 Wave Asset {type=WaveAsset:Array}
]
interfaces [Fort.Character:1.0, UE.OutputFormat.Stereo:1.0, UE.Source.OneShot:1.0, UE.Source:1.1]
outputs [
 "UE.OutputFormat.Stereo.Audio:0" {type=Audio}
 "UE.OutputFormat.Stereo.Audio:1" {type=Audio}
 UE.Source.OneShot.OnFinished {type=Trigger}
]
pages [Default {
  input_defaults {
   Character.Affiliation=0
   Character.Health=0
   Character.UniqueID=0
   MasterVolume=1
   Random.PitchShift.Base=0
   Random.PitchShift.Spread=0.5
   UE.Source.OnPlay=null
   Sounds [/LemonExam/Assets/Audio/Wavs/LemonExam_Pickup_01]
   Wave Asset []
  }
  nodes [MSP_Audio_WavePlayer_Deluxe_Stereo_0 = &MSP_Audio_WavePlayer_Deluxe_Stereo(Play Trigger=^UE.Source.OnPlay, Wave Asset=^Wave Asset)]
  links [
   {dst="@UE.Source.OneShot.OnFinished", src=MSP_Audio_WavePlayer_Deluxe_Stereo_0.On Finished}
   {dst="@UE.OutputFormat.Stereo.Audio:0", src=MSP_Audio_WavePlayer_Deluxe_Stereo_0.Out L}
   {dst="@UE.OutputFormat.Stereo.Audio:1", src=MSP_Audio_WavePlayer_Deluxe_Stereo_0.Out R}
  ]
 }]
properties {
 assetImportData=&self:AssetImportData
 attenuationSettings=&AS_Fort_WeaponFoley
 maxDistance=2250
 numChannels=2
 outputFormat=Stereo
 sampleRate=48000
 soundClassObject=&Fort_WeaponFoley
 assetUserData [&self:AudioAssetUserData_0]
 concurrencySet [&CS_Fort_WeaponFoley]
}
