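# Sound cue for the LemonExam charged attack projectile, combining a launch sound with a "Loop" wave.
# A mixer layers the launch with the quieter, envelope-faded-in loop, each through a default modulator for variation.
# Nested islocal and isenemy branches share that mix, halving volume only for non-local friendly sources.
# It uses the melee weapon sound class, projectile attenuation with NaturalSound falloff overrides, and a two-voice concurrency limit.
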SoundCue /LemonExam/Assets/Audio/SC_LemonExam_Attack_Charged_Proj
root=Branch_2
nodes [
 {id=Branch_2, children [Modulator_2,Branch_3,Modulator_2], properties {boolParameterName=islocal}}
 {id=Branch_3, children [Modulator_4,Modulator_3,Modulator_4], properties {boolParameterName=isenemy}}
 {
  id=Enveloper_0
  children [Modulator_5]
  properties {pitchCurve {editorCurveData {keys [{time=0}]}}, volumeCurve {editorCurveData {keys [<at 0 0>,<at 0.3984 0.1524>,<at 1 1>]}}}
 }
 {id=Mixer_0, children [Modulator_1,Enveloper_0], properties {inputVolume [1,0.5]}}
 {id=Modulator_1, children [WavePlayer_0]}
 {id=Modulator_2, children [Mixer_0], properties {pitchMax=1, pitchMin=1, volumeMax=1, volumeMin=1}}
 {id=Modulator_3, children [Mixer_0], properties {pitchMax=1, pitchMin=1, volumeMax=0.5, volumeMin=0.5}}
 {id=Modulator_4, children [Mixer_0], properties {pitchMax=1, pitchMin=1, volumeMax=1, volumeMin=1}}
 {id=Modulator_5, children [WavePlayer_1]}
 {id=WavePlayer_0, properties {soundWaveAssetPtr=/LemonExam/Assets/Audio/Wavs/LemonExam_Attack_Charged_Proj_Launch_01}}
 {id=WavePlayer_1, properties {soundWaveAssetPtr=/LemonExam/Assets/Audio/Wavs/LemonExam_Attack_Charged_Proj_Loop_01}}
]
properties {
 attenuationSettings=&/ATT_LemonExam_Projectile
 bOverrideConcurrency=true
 maxDistance=6000
 soundClassObject=&Fort_WeaponMelee_PSM
 volumeMultiplier=1.589
 attenuationOverrides {distanceAlgorithm=NaturalSound, falloffDistance=2000, attenuationShapeExtents {x=500}}
 concurrencyOverrides {maxCount=2}
 concurrencySet [null]
}
