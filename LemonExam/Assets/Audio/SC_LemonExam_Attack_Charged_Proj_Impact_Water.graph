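# Sound cue for the water impact of a LemonExam charged attack projectile.
# A Random node picks one of four equally weighted heavy medium-object water splash waves borrowed from the shared physics sounds, then feeds it through a Modulator.
# It uses the projectile impact attenuation and melee weapon sound class, with boosted volume.
# A concurrency override caps simultaneous plays at two, stopping the farthest first.
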SoundCue /LemonExam/Assets/Audio/SC_LemonExam_Attack_Charged_Proj_Impact_Water
root=Modulator_0
nodes [
 {id=Modulator_0, children [Random_3]}
 {id=Random_3, children [WavePlayer_10, WavePlayer_11, WavePlayer_12, WavePlayer_13], properties {weights [1,1,1,1]}}
 {
  id=WavePlayer_10
  properties {soundWaveAssetPtr=/FNGameplayCues/Sounds/Physics/Objects/WaterSplash/MediumObject/Phys_Medium_Object_Splash_Heavy_01}
 }
 {
  id=WavePlayer_11
  properties {soundWaveAssetPtr=/FNGameplayCues/Sounds/Physics/Objects/WaterSplash/MediumObject/Phys_Medium_Object_Splash_Heavy_02}
 }
 {
  id=WavePlayer_12
  properties {soundWaveAssetPtr=/FNGameplayCues/Sounds/Physics/Objects/WaterSplash/MediumObject/Phys_Medium_Object_Splash_Heavy_03}
 }
 {
  id=WavePlayer_13
  properties {soundWaveAssetPtr=/FNGameplayCues/Sounds/Physics/Objects/WaterSplash/MediumObject/Phys_Medium_Object_Splash_Heavy_04}
 }
]
properties {
 attenuationSettings=&/ATT_LemonExam_ProjectileImpact
 bOverrideConcurrency=true
 maxDistance=7000
 soundClassObject=&Fort_WeaponMelee_PSM
 volumeMultiplier=1.5
 attenuationOverrides {distanceAlgorithm=NaturalSound, falloffDistance=2000, attenuationShapeExtents {x=500}}
 concurrencyOverrides {maxCount=2, resolutionRule=StopFarthestThenPreventNew}
 concurrencySet [null]
}
