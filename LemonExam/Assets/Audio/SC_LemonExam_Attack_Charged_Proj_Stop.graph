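# Sound cue for the LemonExam charged attack projectile stop, playing one of three stop sound variations.
# A Random node with equal weights picks among three wave players and feeds a Modulator at the root, left at its default settings.
# It uses a projectile-impact attenuation asset and the melee weapon sound class, with a volume boost and an overridden concurrency that caps simultaneous instances at a low count.
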SoundCue /LemonExam/Assets/Audio/SC_LemonExam_Attack_Charged_Proj_Stop
root=Modulator_0
nodes [
 {id=Modulator_0, children [Random_0]}
 {id=Random_0, children [WavePlayer_2,WavePlayer_3,WavePlayer_4], properties {weights [1,1,1]}}
 {id=WavePlayer_2, properties {soundWaveAssetPtr=/LemonExam/Assets/Audio/Wavs/LemonExam_Attack_Projectile_Stop_01}}
 {id=WavePlayer_3, properties {soundWaveAssetPtr=/LemonExam/Assets/Audio/Wavs/LemonExam_Attack_Projectile_Stop_02}}
 {id=WavePlayer_4, properties {soundWaveAssetPtr=/LemonExam/Assets/Audio/Wavs/LemonExam_Attack_Projectile_Stop_03}}
]
properties {
 attenuationSettings=&/ATT_LemonExam_ProjectileImpact
 bOverrideConcurrency=true
 maxDistance=7000
 soundClassObject=&Fort_WeaponMelee_PSM
 volumeMultiplier=1.5
 attenuationOverrides {distanceAlgorithm=NaturalSound, falloffDistance=2000, attenuationShapeExtents {x=500}}
 concurrencyOverrides {maxCount=2}
 concurrencySet [null]
}
