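# Sound cue for the whoosh-by of a LemonExam charged attack projectile.
# A Random node picks one of three equally weighted whoosh-by wave variations and feeds it through a Modulator.
# It uses the LemonExam projectile attenuation asset and a melee weapon sound class, with a concurrency override capping simultaneous instances at two.
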SoundCue /LemonExam/Assets/Audio/SC_LemonExam_Attack_Charged_Proj_Whoosh
root=Modulator_0
nodes [
 {id=Modulator_0, children [Random_1]}
 {id=Random_1, children [WavePlayer_0,WavePlayer_1,WavePlayer_5], properties {weights [1,1,1]}}
 {id=WavePlayer_0, properties {soundWaveAssetPtr=/LemonExam/Assets/Audio/Wavs/LemonExam_Attack_Projectile_WhooshBy_01}}
 {id=WavePlayer_1, properties {soundWaveAssetPtr=/LemonExam/Assets/Audio/Wavs/LemonExam_Attack_Projectile_WhooshBy_02}}
 {id=WavePlayer_5, properties {soundWaveAssetPtr=/LemonExam/Assets/Audio/Wavs/LemonExam_Attack_Projectile_WhooshBy_03}}
]
properties {
 attenuationSettings=&/ATT_LemonExam_Projectile
 bOverrideConcurrency=true
 maxDistance=6000
 soundClassObject=&Fort_WeaponMelee_PSM
 volumeMultiplier=1
 attenuationOverrides {distanceAlgorithm=NaturalSound, falloffDistance=2000, attenuationShapeExtents {x=500}}
 concurrencyOverrides {maxCount=2}
 concurrencySet [null]
}
