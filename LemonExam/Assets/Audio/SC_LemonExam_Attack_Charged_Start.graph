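# Sound cue for the start of the LemonExam charged attack, a melee weapon sound built on a single wave.
# A shared modulator adds slight random variation to the wave before branches on islocal and isenemy set its level.
# Non-local, non-enemy sources play at half volume, while local, enemy and unset cases play at full volume.
# It uses charged-attack attenuation, a melee sound class and player melee weapon concurrency.
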SoundCue /LemonExam/Assets/Audio/SC_LemonExam_Attack_Charged_Start
root=Branch_2
nodes [
 {id=Branch_2, children [Modulator_2,Branch_3,Modulator_2], properties {boolParameterName=islocal}}
 {id=Branch_3, children [Modulator_4,Modulator_3,Modulator_4], properties {boolParameterName=isenemy}}
 {id=Modulator_0, children [WavePlayer_0]}
 {id=Modulator_2, children [Modulator_0], properties {pitchMax=1, pitchMin=1, volumeMax=1, volumeMin=1}}
 {id=Modulator_3, children [Modulator_0], properties {pitchMax=1, pitchMin=1, volumeMax=0.5, volumeMin=0.5}}
 {id=Modulator_4, children [Modulator_0], properties {pitchMax=1, pitchMin=1, volumeMax=1, volumeMin=1}}
 {id=WavePlayer_0, properties {soundWaveAssetPtr=/LemonExam/Assets/Audio/Wavs/LemonExam_Attack_Charged_Start_01}}
]
properties {
 attenuationSettings=&/ATT_LemonExam_Attack_Charged
 maxDistance=5750
 soundClassObject=&Fort_WeaponMelee_PSM
 volumeMultiplier=1.589
 attenuationOverrides {distanceAlgorithm=NaturalSound, falloffDistance=2000, attenuationShapeExtents {x=500}}
 concurrencySet [&Player_Melee_Weapon_Conc]
}
