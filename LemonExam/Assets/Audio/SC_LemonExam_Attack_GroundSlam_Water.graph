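# Sound cue for the water splash of a ground slam attack in LemonExam.
# A modulator feeds an equal-weight random pick among four medium large-object water splash waves borrowed from the shared physics sounds.
# It plays through the melee weapon sound class with the LemonExam attack attenuation asset and overrides concurrency to limit simultaneous instances.
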SoundCue /LemonExam/Assets/Audio/SC_LemonExam_Attack_GroundSlam_Water
root=Modulator_0
nodes [
 {id=Modulator_0, children [Random_3]}
 {id=Random_3, children [WavePlayer_10, WavePlayer_11, WavePlayer_12, WavePlayer_13], properties {weights [1,1,1,1]}}
 {
  id=WavePlayer_10
  properties {soundWaveAssetPtr=/FNGameplayCues/Sounds/Physics/Objects/WaterSplash/LargeObject/Phys_Large_Object_Splash_Medium_01}
 }
 {
  id=WavePlayer_11
  properties {soundWaveAssetPtr=/FNGameplayCues/Sounds/Physics/Objects/WaterSplash/LargeObject/Phys_Large_Object_Splash_Medium_02}
 }
 {
  id=WavePlayer_12
  properties {soundWaveAssetPtr=/FNGameplayCues/Sounds/Physics/Objects/WaterSplash/LargeObject/Phys_Large_Object_Splash_Medium_03}
 }
 {
  id=WavePlayer_13
  properties {soundWaveAssetPtr=/FNGameplayCues/Sounds/Physics/Objects/WaterSplash/LargeObject/Phys_Large_Object_Splash_Medium_04}
 }
]
properties {
 attenuationSettings=&/ATT_LemonExam_Attacks
 bOverrideConcurrency=true
 maxDistance=5500
 soundClassObject=&Fort_WeaponMelee_PSM
 volumeMultiplier=1
 attenuationOverrides {distanceAlgorithm=NaturalSound, falloffDistance=2000, attenuationShapeExtents {x=500}}
 concurrencyOverrides {maxCount=2}
}
