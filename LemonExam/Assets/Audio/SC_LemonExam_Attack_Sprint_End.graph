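# Sound cue for the end of the LemonExam sprint attack, randomly playing one of four equally weighted waves through a shared Modulator.
# Branches on "islocal" and "isenemy" route that shared chain through volume modulators, so non-local, non-enemy sources play at half volume while the others stay at full.
# It uses the LemonExam attack attenuation with overrides, the melee weapon sound class, a raised pitch and a two-voice concurrency limit.
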SoundCue /LemonExam/Assets/Audio/SC_LemonExam_Attack_Sprint_End
root=Branch_2
nodes [
 {id=Branch_2, children [Modulator_2,Branch_3,Modulator_2], properties {boolParameterName=islocal}}
 {id=Branch_3, children [Modulator_4,Modulator_3,Modulator_4], properties {boolParameterName=isenemy}}
 {id=Modulator_1, children [Random_1]}
 {id=Modulator_2, children [Modulator_1], properties {pitchMax=1, pitchMin=1, volumeMax=1, volumeMin=1}}
 {id=Modulator_3, children [Modulator_1], properties {pitchMax=1, pitchMin=1, volumeMax=0.5, volumeMin=0.5}}
 {id=Modulator_4, children [Modulator_1], properties {pitchMax=1, pitchMin=1, volumeMax=1, volumeMin=1}}
 {id=Random_1, children [WavePlayer_2,WavePlayer_3,WavePlayer_6,WavePlayer_7], properties {weights [1,1,1,1]}}
 {id=WavePlayer_2, properties {soundWaveAssetPtr=/LemonExam/Assets/Audio/Wavs/LemonExam_Sprint_End_01}}
 {id=WavePlayer_3, properties {soundWaveAssetPtr=/LemonExam/Assets/Audio/Wavs/LemonExam_Sprint_End_02}}
 {id=WavePlayer_6, properties {soundWaveAssetPtr=/LemonExam/Assets/Audio/Wavs/LemonExam_Sprint_End_03}}
 {id=WavePlayer_7, properties {soundWaveAssetPtr=/LemonExam/Assets/Audio/Wavs/LemonExam_Sprint_End_04}}
]
properties {
 attenuationSettings=&/ATT_LemonExam_Attacks
 bOverrideConcurrency=true
 maxDistance=5500
 pitchMultiplier=1.2
 soundClassObject=&Fort_WeaponMelee_PSM
 volumeMultiplier=1
 attenuationOverrides {distanceAlgorithm=NaturalSound, falloffDistance=2000, attenuationShapeExtents {x=500}}
 concurrencyOverrides {maxCount=2}
}
