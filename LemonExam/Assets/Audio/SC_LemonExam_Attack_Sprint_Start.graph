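# Sound cue for the start of the LemonExam sprint attack, under a melee weapon sound class with attenuation and a concurrency limit.
# A random start one-shot, slightly quieter, mixes with a looping sprint sound whose volume envelope fades in, holds, then eases down.
# Branches on the islocal and isenemy parameters share that mix, halving volume for non-local, non-enemy sources and otherwise keeping it full.
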SoundCue /LemonExam/Assets/Audio/SC_LemonExam_Attack_Sprint_Start
root=Branch_2
nodes [
 {id=Branch_2, children [Modulator_2,Branch_3,Modulator_2], properties {boolParameterName=islocal}}
 {id=Branch_3, children [Modulator_4,Modulator_3,Modulator_4], properties {boolParameterName=isenemy}}
 {
  id=Enveloper_0
  children [Modulator_0]
  properties {pitchCurve {editorCurveData {keys [{time=0}]}}, volumeCurve {editorCurveData {keys [<at 0 0>,<at 1 1>,<at 4 1>,<at 10 0.6>]}}}
 }
 {id=Mixer_0, children [Modulator_1,Enveloper_0], properties {inputVolume [0.7,1]}}
 {id=Modulator_0, children [WavePlayer_3]}
 {id=Modulator_1, children [Random_0]}
 {id=Modulator_2, children [Mixer_0], properties {pitchMax=1, pitchMin=1, volumeMax=1, volumeMin=1}}
 {id=Modulator_3, children [Mixer_0], properties {pitchMax=1, pitchMin=1, volumeMax=0.5, volumeMin=0.5}}
 {id=Modulator_4, children [Mixer_0], properties {pitchMax=1, pitchMin=1, volumeMax=1, volumeMin=1}}
 {id=Random_0, children [WavePlayer_0,WavePlayer_1,WavePlayer_4,WavePlayer_5], properties {weights [1,1,1,1]}}
 {id=WavePlayer_0, properties {soundWaveAssetPtr=/LemonExam/Assets/Audio/Wavs/LemonExam_Sprint_Start_01}}
 {id=WavePlayer_1, properties {soundWaveAssetPtr=/LemonExam/Assets/Audio/Wavs/LemonExam_Sprint_Start_02}}
 {id=WavePlayer_3, properties {bLooping=true, soundWaveAssetPtr=/LemonExam/Assets/Audio/Wavs/LemonExam_Sprint_Loop_01}}
 {id=WavePlayer_4, properties {soundWaveAssetPtr=/LemonExam/Assets/Audio/Wavs/LemonExam_Sprint_Start_03}}
 {id=WavePlayer_5, properties {soundWaveAssetPtr=/LemonExam/Assets/Audio/Wavs/LemonExam_Sprint_Start_04}}
]
properties {
 attenuationSettings=&/ATT_LemonExam_Attacks
 bOverrideConcurrency=true
 maxDistance=5500
 soundClassObject=&Fort_WeaponMelee_PSM
 volumeMultiplier=1
 attenuationOverrides {distanceAlgorithm=NaturalSound, falloffDistance=2000, attenuationShapeExtents {x=500}}
 concurrencyOverrides {maxCount=2}
}
