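# Melee impact SoundCue for LemonExam that plays one of four normal attack impact waves.
# An equally weighted Random node picks the wave, then a Modulator at default settings feeds the output.
# The cue uses a dedicated melee impact attenuation asset, a melee weapon sound class and a boosted volume.
# Its concurrency override caps simultaneous instances and blocks new plays at the limit, taking precedence over the referenced player melee concurrency set.
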SoundCue /LemonExam/Assets/Audio/SC_LemonExam_Melee_Impt_01
root=Modulator_0
nodes [
 {id=Modulator_0, children [Random_1]}
 {id=Random_1, children [WavePlayer_0,WavePlayer_1,WavePlayer_6,WavePlayer_7], properties {weights [1,1,1,1]}}
 {id=WavePlayer_0, properties {soundWaveAssetPtr=/LemonExam/Assets/Audio/Wavs/LemonExam_Attack_Impact_Normal_01}}
 {id=WavePlayer_1, properties {soundWaveAssetPtr=/LemonExam/Assets/Audio/Wavs/LemonExam_Attack_Impact_Normal_02}}
 {id=WavePlayer_6, properties {soundWaveAssetPtr=/LemonExam/Assets/Audio/Wavs/LemonExam_Attack_Impact_Normal_03}}
 {id=WavePlayer_7, properties {soundWaveAssetPtr=/LemonExam/Assets/Audio/Wavs/LemonExam_Attack_Impact_Normal_04}}
]
properties {
 attenuationSettings=&/ATT_LemonExam_Melee_Impacts
 bOverrideConcurrency=true
 maxDistance=2500
 soundClassObject=&Fort_WeaponMelee_PSM
 volumeMultiplier=1.5
 concurrencyOverrides {maxCount=2, resolutionRule=PreventNew, retriggerTime=0.25}
 concurrencySet [&Player_Melee_Weapon_Conc]
}
