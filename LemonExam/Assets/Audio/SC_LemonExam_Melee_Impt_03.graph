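# Sound cue for LemonExam melee weapon heavy attack impacts.
# A Random node picks one of three equally weighted heavy impact waves and feeds it through a Modulator for variation.
# It uses a melee impact attenuation asset, the weapon melee sound class and boosted volume.
# A concurrency override caps simultaneous plays, blocks new ones at the limit and adds a retrigger time, despite an assigned player melee concurrency set.
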SoundCue /LemonExam/Assets/Audio/SC_LemonExam_Melee_Impt_03
root=Modulator_0
nodes [
 {id=Modulator_0, children [Random_1]}
 {id=Random_1, children [WavePlayer_0,WavePlayer_1,WavePlayer_2], properties {weights [1,1,1]}}
 {id=WavePlayer_0, properties {soundWaveAssetPtr=/LemonExam/Assets/Audio/Wavs/LemonExam_Attack_Impact_Heavy_01}}
 {id=WavePlayer_1, properties {soundWaveAssetPtr=/LemonExam/Assets/Audio/Wavs/LemonExam_Attack_Impact_Heavy_02}}
 {id=WavePlayer_2, properties {soundWaveAssetPtr=/LemonExam/Assets/Audio/Wavs/LemonExam_Attack_Impact_Heavy_03}}
]
properties {
 attenuationSettings=&/ATT_LemonExam_Melee_Impacts
 bOverrideConcurrency=true
 maxDistance=2500
 soundClassObject=&Fort_WeaponMelee_PSM
 volumeMultiplier=2
 concurrencyOverrides {maxCount=2, resolutionRule=PreventNew, retriggerTime=0.25}
 concurrencySet [&Player_Melee_Weapon_Conc]
}
