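# Sound cue for the third LemonExam melee swing, playing one of three equally weighted swing waves with slight random pitch variation.
# Bool branches on islocal, then isenemy, halve the volume for non-local, non-enemy swingers; local, enemy and unset cases stay at full volume.
# It uses the LemonExam attack attenuation, the melee weapon sound class and player melee weapon concurrency.
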SoundCue /LemonExam/Assets/Audio/SC_LemonExam_Melee_Swing_03
root=Branch_2
nodes [
 {id=Branch_2, children [Modulator_2,Branch_3,Modulator_2], properties {boolParameterName=islocal}}
 {id=Branch_3, children [Modulator_4,Modulator_3,Modulator_4], properties {boolParameterName=isenemy}}
 {id=Modulator_0, children [Random_1], properties {pitchMax=1.1, pitchMin=0.9}}
 {id=Modulator_2, children [Modulator_0], properties {pitchMax=1, pitchMin=1, volumeMax=1, volumeMin=1}}
 {id=Modulator_3, children [Modulator_0], properties {pitchMax=1, pitchMin=1, volumeMax=0.5, volumeMin=0.5}}
 {id=Modulator_4, children [Modulator_0], properties {pitchMax=1, pitchMin=1, volumeMax=1, volumeMin=1}}
 {id=Random_1, children [WavePlayer_3,WavePlayer_4,WavePlayer_5], properties {weights [1,1,1]}}
 {id=WavePlayer_3, properties {soundWaveAssetPtr=/LemonExam/Assets/Audio/Wavs/LemonExam_Attack_Swing_03_01}}
 {id=WavePlayer_4, properties {soundWaveAssetPtr=/LemonExam/Assets/Audio/Wavs/LemonExam_Attack_Swing_03_02}}
 {id=WavePlayer_5, properties {soundWaveAssetPtr=/LemonExam/Assets/Audio/Wavs/LemonExam_Attack_Swing_03_03}}
]
properties {
 attenuationSettings=&/ATT_LemonExam_Attacks
 maxDistance=5500
 soundClassObject=&Fort_WeaponMelee_PSM
 volumeMultiplier=1.589
 attenuationOverrides {distanceAlgorithm=NaturalSound, falloffDistance=2000, attenuationShapeExtents {x=500}}
 concurrencySet [&Player_Melee_Weapon_Conc]
}
