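# Masked master material for the LemonExam Sharp blade, layering an animated glowing surface over the shared 2023 character PreFX/PostFX functions.
# Noise-scrolled gradient maps, parallaxed patterns, reflection and fresnel terms build time-of-day-scaled emissive, with masked blending into subsurface shading.
# Blueprint-fed start times or stage parameters drive attack (ATK) and appear (APR) animations, shaped partly by a curve atlas.
# A noisy shifting mask produces a dithered dissolve with glowing edges.
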Material /LemonExam/Assets/Materials/M_LemonExam_Sharp
graph {
 properties {
  bUseMaterialAttributes=true
  bUsedWithClothing=true
  bUsedWithMeshDeformer=true
  bUsedWithMorphTargets=true
  bUsedWithNanite=true
  bUsedWithNiagaraMeshParticles=true
  bUsedWithSkeletalMesh=true
  blendMode=BLEND_Masked
  pixelDepthOffsetMode=PDOM_Legacy
  shadingModel=MSM_FromMaterialExpression
  usedShadingModels=Default Lit | Subsurface | Clear Coat | Subsurface Profile | Cloth
 }
 nodes [
  Constant() {r=1}
  Constant() {r=1}
  Divide(A=Time()) {constB=5}
  Frac(None=Multiply(A=Time()) {constB=0.2})
  n1 = GetMaterialAttributes(None=NamedRerouteUsage()) {attributeGetTypes [
    #1
    #2
    #3
    #4
   ]}
  n2 = NamedRerouteUsage()
  n3 = SetMaterialAttributes!(
   MaterialAttributes=@n1.MaterialAttributes
   Emissive Color=NamedRerouteUsage()
   Base Color=NamedRerouteUsage()
   Fuzz Color or Subsurface Color=NamedRerouteUsage()
   Opacity=NamedRerouteUsage()
   Specular=LinearInterpolate(A=@n1.Specular, B=ScalarParameter() {defaultValue=2, group=Quail Wink / Loan Float / Evoke Find, parameterName=Specular Blade}, Alpha=@n2) {constB=0}
   Metallic=LinearInterpolate(A=@n1.Metallic, Alpha=@n2)
  ) {attributeSetTypes [
    #1
    #2
    #5
    #6
    #3
    #4
   ]}
  n4 = ShadingModel() {shadingModel=MSM_Subsurface}
  n5 = GetMaterialAttributes(None=BlendMaterialAttributes(
    A=@n3
    B=SetMaterialAttributes!(MaterialAttributes=@n3, Shading Model=@n4) {attributeSetTypes [#7]}
    Alpha=NamedRerouteUsage()
   )) {attributeGetTypes [#1]}
  n6 = MaterialFunctionCall() {materialFunction=&BoundingBoxBased_0-1_UVW}
  n7 = OneMinus(None=NamedRerouteUsage())
  n8 = Step(Y=@n7) {constX=0.01}
  n9 = Multiply(A=LinearInterpolate(A=OneMinus(None=@n6.B), B=@n6.B, Alpha=@n8)) {constB=0.8}
  n10 = Multiply(A=NamedRerouteUsage(), B=@n7)
  n11 = LinearInterpolate(A=OneMinus(None=@n10), B=@n10, Alpha=@n8)
  n12 = Constant4Vector() {constant=<rgba 0 0.1 0.1 1>}
  n13 = MaterialFunctionCall!(Mask In=@n9, Amount=@n11, Range=@n12.G, Start Offset=@n12.G, Sharpness=@n12.G) {materialFunction=/Game/Global/MaterialFunctions/MF_MaskShifting}
  n14 = Panner(Coordinate=Multiply(A=NamedRerouteUsage(), B=Constant2Vector() {g=8, r=0.25}), Time=@n10) {speedX=-0.5, speedY=1}
  n15 = NamedRerouteUsage()
  n16 = TextureSample(UVs=@n14, Tex=@n15) {samplerSource=SSM_Wrap_WorldGroupSettings, samplerType=SAMPLERTYPE_LinearColor}
  n17 = TextureSample(UVs=Multiply(A=@n14) {constB=0.5}, Tex=@n15) {samplerSource=SSM_Wrap_WorldGroupSettings, samplerType=SAMPLERTYPE_LinearColor}
  n18 = Add(A=Multiply(A=@n13.Result, B=Multiply(A=@n16.R, B=@n17.R)), B=@n13.Result)
  n19 = MaterialFunctionCall(Mask In=@n9, Amount=@n11, Range=Multiply(A=@n12.G) {constB=10}, Start Offset=@n12.B, Sharpness=@n12.B) {materialFunction=&MF_MaskShifting}
  n20 = Multiply(A=NamedRerouteUsage()) {constB=0.75}
  n21 = LinearInterpolate(A=@n18, B=@n19.Result, Alpha=@n20) {constA=5, constB=40}
  n22 = Saturate(None=Multiply(A=OneMinus(None=@n21), B=@n21))
  n23 = LinearInterpolate(A=@n13.Result, B=@n19.Result, Alpha=@n20) {constA=5, constB=40}
  n24 = Saturate(None=Multiply(A=OneMinus(None=@n23), B=@n23))
  n25 = LinearInterpolate(Alpha=NamedRerouteUsage()) {constA=0.5, constB=20}
  n26 = CurveAtlasRowParameter(CurveTime=@n10) {
   atlas=/LemonExam/Assets/Textures/CA_LemonExam_Animation
   curve=/LemonExam/Assets/Textures/C_LemonExam_Animation
   parameterName=LemonExamAnimationCurve
  }
  n27 = Multiply(A=Max(A=Multiply(A=@n22) {constB=10}, B=Multiply(A=@n24, B=@n25) {constB=10}), B=@n26.B)
  n28 = Saturate(None=@n18)
  n29 = MaterialFunctionCall() {materialFunction=&MF_CharacterParent_2023_PreFX}
  MaterialFunctionCall(
   Material=SetMaterialAttributes!(
    MaterialAttributes=@n5.MaterialAttributes
    Emissive Color=Add(A=@n5.Emissive Color, B=@n27)
    Opacity Mask=MaterialFunctionCall(Alpha Threshold=@n28) {materialFunction=&DitherTemporalAA_Fortnite}.Result
   ) {attributeSetTypes [#1, #8]}
   M=@n29.M
   Pre Skinned Local Position=@n29.M
   Distance Blend=@n29.M
   Fresnel=@n29.M
  ) {materialFunction=&MF_CharacterParent_2023_PostFX}
  n30 = TextureSample() {desc=Tech Art Mask, samplerSource=SSM_Wrap_WorldGroupSettings, texture=/LemonExam/Assets/Textures/T_LemonExam_Sharp_EdgeMask}
  NamedRerouteDeclaration(None=Multiply(A=@n30.R, B=ScalarParameter() {defaultValue=1, group=Quail Wink / Loan Float / Evoke Find, parameterName=Thickness Opacity})) {name=Edge, nodeColor=<rgba 0.1484 1 0 1>}
  n31 = NamedRerouteUsage()
  n32 = Multiply(A=SphereMask(A=NamedRerouteUsage(), B=Constant3Vector() {constant=<rgba 0 -70 5 0>}, Radius=LinearInterpolate(Alpha=@n31) {constB=160}) {attenuationRadius=160, hardnessPercent=0}) {constB=10}
  NamedRerouteDeclaration(None=Multiply(A=Saturate(None=@n32), B=NamedRerouteUsage())) {name=AttackGlowing_Msk, nodeColor=<rgba 1 0.4988 0.4323 1>}
  n33 = NamedRerouteUsage()
  n34 = ScalarParameter() {defaultValue=0.7, group=Quail Wink / Loan Float / Evoke Find, parameterName=Flow Noise - Tiling}
  n35 = TextureSample(
   UVs=Panner(Coordinate=Multiply(A=@n33, B=@n34)) {speedX=0.05, speedY=0.1}
   Tex=TextureObject() {samplerType=SAMPLERTYPE_LinearColor, texture=/Game/Global/Textures/Default/Tileable_Patterns/T_PerlinNoise_Uncompressed}
  ) {
   samplerSource=SSM_Wrap_WorldGroupSettings
   samplerType=SAMPLERTYPE_LinearColor
   texture=&T_PerlinNoise_Uncompressed
  }
  n36 = Multiply(
   A=Power(Base=@n35.R, Exp=ScalarParameter() {group=Quail Wink / Loan Float / Evoke Find, parameterName=Flow Noise - Exponent})
   B=ScalarParameter() {defaultValue=1, group=Quail Wink / Loan Float / Evoke Find, parameterName=Flow Noise - Strength}
  )
  n37 = Step(
   Y=ScalarParameter() {defaultValue=0.1, group=Quail Wink / Loan Float / Evoke Find, parameterName=Flow Mask Edge Remvoal}
   X=NamedRerouteUsage()
  ) {constY=0.8}
  NamedRerouteDeclaration(None=Multiply(A=Saturate(None=@n36), B=@n37)) {name=Pattern Flow, nodeColor=<rgba 0 1 0.9766 1>}
  n38 = LinearInterpolate(
   A=ScalarParameter() {defaultValue=0.3, group=Quail Wink / Loan Float / Evoke Find, parameterName=SSS Strength}
   B=ScalarParameter() {defaultValue=0.7, group=Quail Wink / Loan Float / Evoke Find, parameterName=Blade Edge - SSS Strength}
   Alpha=NamedRerouteUsage()
  )
  NamedRerouteDeclaration(None=LinearInterpolate(B=@n38, Alpha=NamedRerouteUsage()) {constA=1, constB=0}) {name=SSS Opacity, nodeColor=<rgba 1 0.6563 0 1>}
  n39 = NamedRerouteUsage()
  n40 = NamedRerouteUsage()
  n41 = Multiply(A=NamedRerouteUsage(), B=ScalarParameter() {group=Quail Wink / Loan Float / Evoke Find, parameterName=SSS Min})
  NamedRerouteDeclaration(None=Multiply(A=Multiply(A=@n39, B=@n38), B=Max(A=@n40, B=@n41))) {name=SSS Color, nodeColor=<rgba 0 1 0.6484 1>}
  NamedRerouteDeclaration(None=Multiply(
    A=NamedRerouteDeclaration(None=@n37) {name=Pattern Mask Out, nodeColor=<rgba 0.4297 0 1 1>}
    B=ScalarParameter() {defaultValue=0.25, group=Quail Wink / Loan Float / Evoke Find, parameterName=Pattern - Opacity}
   )) {name=Pattern Alpha, nodeColor=<rgba 1 0.3281 0 1>}
  NamedRerouteDeclaration(None=@n1.Base Color) {name=Base Color In, nodeColor=<rgba 0.4766 1 0 1>}
  n42 = Subtract(A=Time(), B=ScalarParameter() {defaultValue=-99999, group=Quail Wink / Loan Float / Evoke Find, parameterName=ATK_GameTimeFromBP})
  n43 = Add(A=ScalarParameter() {defaultValue=1, group=Quail Wink / Loan Float / Evoke Find, parameterName=ATK_AnimationDuration, sortPriority=101}) {constB=0}
  NamedRerouteDeclaration(None=Saturate(None=Divide(A=@n42, B=@n43))) {name=ATK_TimeNative, nodeColor=<rgba 0.2891 1 0 1>}
  n44 = ScalarParameter() {group=Quail Wink / Loan Float / Evoke Find, parameterName=ATK_Stage}
  n45 = LinearInterpolate(A=Subtract(A=@n44), B=@n44, Alpha=NamedRerouteUsage()) {constB=20}
  n46 = CurveAtlasRowParameter(CurveTime=Saturate(None=@n45)) {
   atlas=&CA_LemonExam_Animation
   curve=&C_LemonExam_Animation
   parameterName=LemonExamAnimationCurve
  }
  NamedRerouteDeclaration(None=@n46.R) {name=ATK_Stages_01, nodeColor=<rgba 0.3812 1 0.0069 1>}
  n47 = Constant4Vector() {constant=<rgba 1 2 0 1>}
  NamedRerouteDeclaration(None=Saturate(None=MaterialFunctionCall(Input=@n45, Input Low=@n47.R, Input High=@n47.R, Target Low=@n47.R, Target High=@n47.R) {materialFunction=&RemapValueRange}.Result)) {name=ATK_Stages_02, nodeColor=<rgba 0.3812 1 0.0069 1>}
  n48 = NamedRerouteUsage()
  n49 = ScalarParameter() {defaultValue=0.05, group=Quail Wink / Loan Float / Evoke Find, parameterName=Base Color}
  n50 = NamedRerouteUsage()
  n51 = NamedRerouteUsage()
  NamedRerouteDeclaration(None=LinearInterpolate(A=NamedRerouteUsage(), B=Multiply(A=@n48, B=@n49), Alpha=Max(A=@n50, B=@n51))) {name=Base Color, nodeColor=<rgba 1 0 0.7969 1>}
  n52 = NamedRerouteUsage()
  n53 = NamedRerouteUsage()
  n54 = NamedRerouteUsage()
  n55 = NamedRerouteUsage()
  n56 = NamedRerouteUsage()
  n57 = NamedRerouteUsage()
  n58 = Max(
   A=Multiply(A=Multiply(A=@n54, B=@n55), B=ScalarParameter() {defaultValue=2, group=Quail Wink / Loan Float / Evoke Find, parameterName=Pattern - Min})
   B=Multiply(A=Multiply(A=@n56, B=@n57), B=ScalarParameter() {defaultValue=6, group=Quail Wink / Loan Float / Evoke Find, parameterName=Pattern - Max})
  )
  n59 = MaterialFunctionCall(Exponent=ScalarParameter() {defaultValue=4, group=Quail Wink / Loan Float / Evoke Find, parameterName=Pattern Fresnel}) {materialFunction=&AxisAlignedFresnel}
  n60 = LinearInterpolate(
   A=LinearInterpolate(A=LinearInterpolate(A=@n52, Alpha=@n53) {constB=0}, B=NamedRerouteUsage(), Alpha=NamedRerouteUsage())
   B=LinearInterpolate(A=@n58, B=Multiply(A=@n58) {constB=2}, Alpha=NamedRerouteUsage()) {constB=0}
   Alpha=Multiply(A=NamedRerouteUsage(), B=@n59.Fresnel)
  )
  n61 = NamedRerouteUsage()
  n62 = Multiply(A=Multiply(A=OneMinus(None=@n61), B=NamedRerouteUsage()), B=Constant3Vector() {constant=<rgba 0.0448 0.1862 0.2396 1>})
  n63 = ScalarParameter() {defaultValue=16, parameterName=Pattern - AttackIntensity}
  n64 = Multiply(A=Multiply(A=Multiply(A=@n62, B=@n63) {constB=20}, B=@n59.Fresnel), B=NamedRerouteUsage())
  n65 = VectorParameter() {parameterName=TOD_Intensity, channelNames {a=Night, b=Evening, g=Day, r=Morning}, defaultValue=<rgba 1 1 1 1>}
  NamedRerouteDeclaration(None=Multiply(
    A=Add(A=@n60, B=@n64)
    B=MaterialFunctionCall(Value_Morning=@n65.Morning, Value_Day=@n65.Morning, Value_Evening=@n65.Morning, Value_Night=@n65.Morning) {materialFunction=&Scalar_Time_of_Day}.Result
   )) {name=Emissive, nodeColor=<rgba 0 1 0.8828 1>}
  NamedRerouteDeclaration(None=@n29.Result) {name=Material Attributes, nodeColor=<rgba 1 0.5625 0 1>}
  NamedRerouteDeclaration(None=GetMaterialAttributes(None=@n29.Result) {attributeGetTypes [#9]}.Roughness) {name=Roughness, nodeColor=<rgba 0.1484 1 0 1>}
  n66 = TextureSample() {samplerSource=SSM_Wrap_WorldGroupSettings, samplerType=SAMPLERTYPE_Masks, texture=/LemonExam/Assets/Textures/T_LemonExam_Sharp_FX}
  NamedRerouteDeclaration(None=@n66.RGB) {name=Debug Masks, nodeColor=<rgba 0.0312 0 1 1>}
  n67 = ScalarParameter() {defaultValue=0.01, group=Quail Wink / Loan Float / Evoke Find, parameterName=Reflection Vector - Strength, sliderMin=18.9627}
  n68 = VertexNormalWS()
  n69 = ReflectionVectorWS(CustomWorldNormal=@n68)
  n70 = CameraVectorWS()
  n71 = StaticComponentMaskParameter(None=MaterialFunctionCall!(
    Normal=Transform(None=@n69) {transformSourceType=TRANSFORMSOURCE_World, transformType=TRANSFORM_Local}
    CameraVector=Transform(None=@n70) {transformSourceType=TRANSFORMSOURCE_World, transformType=TRANSFORM_Local}
   ) {materialFunction=/Engine/Functions/Engine_MaterialFunctions02/Utility/CustomReflectionVector}.Reflection Vector) {defaultR=true, group=Quail Wink / Loan Float / Evoke Find, parameterName=Reflection Axis}
  n72 = ReflectionVectorWS(CustomWorldNormal=@n68)
  n73 = OneMinus(None=CameraVectorWS())
  n74 = StaticComponentMaskParameter(None=MaterialFunctionCall(
    Normal=Transform(None=@n72) {transformSourceType=TRANSFORMSOURCE_World, transformType=TRANSFORM_Local}
    CameraVector=Transform(None=@n73) {transformSourceType=TRANSFORMSOURCE_World, transformType=TRANSFORM_Local}
   ) {materialFunction=&CustomReflectionVector}.Reflection Vector) {defaultG=true, group=Quail Wink / Loan Float / Evoke Find, parameterName=Inverted Reflection Axis}
  n75 = Power(
   Base=StaticSwitchParameter(
    True=Max(A=@n71, B=@n74)
    False=MaterialFunctionCall() {materialFunction=&AxisAlignedFresnel}.Fresnel
   ) {group=Quail Wink / Loan Float / Evoke Find, parameterName=Use Reflection Vector}
   Exp=ScalarParameter() {defaultValue=2.6918, parameterName=Reflection Vector - Exponent}
  )
  n76 = MaterialFunctionCall(Exponent=ScalarParameter() {defaultValue=2.6918, parameterName=Reflection Vector - Exponent}, Normal=PixelNormalWS()) {materialFunction=&AxisAlignedFresnel}
  n77 = Multiply(A=NamedRerouteUsage(), B=Multiply(A=OneMinus(None=@n76.Fresnel), B=NamedRerouteUsage()))
  n78 = ComponentMask(None=ReflectionVectorWS()) {b=true}
  n79 = Saturate(None=Power(
    Base=OneMinus(None=@n78)
    Exp=ScalarParameter() {defaultValue=0.1, group=Quail Wink / Loan Float / Evoke Find, parameterName=Reflection Vector - Y Exponent}
   ))
  n80 = Multiply(
   A=LinearInterpolate(
    A=@n67
    B=ScalarParameter() {defaultValue=3, group=Quail Wink / Loan Float / Evoke Find, parameterName=Edge Reflection - Strength}
    Alpha=NamedRerouteUsage()
   )
   B=Multiply(
    A=Saturate(None=@n75)
    B=StaticSwitchParameter(True=@n77, False=@n79) {defaultValue=true, group=Quail Wink / Loan Float / Evoke Find, parameterName=Evoke Find Reflection}
   )
  )
  n81 = StaticSwitchParameter(True=@n80, False=Saturate(None=@n80)) {defaultValue=true, group=Quail Wink / Loan Float / Evoke Find, parameterName=Evoke Find Reflection}
  n82 = ScalarParameter() {defaultValue=10, group=Quail Wink / Loan Float / Evoke Find, parameterName=Roughness - Base}
  n83 = ScalarParameter() {defaultValue=4, group=Quail Wink / Loan Float / Evoke Find, parameterName=Roughness - Edge}
  n84 = NamedRerouteUsage()
  n85 = Saturate(None=Multiply(A=LinearInterpolate(A=@n82, B=@n83, Alpha=@n84), B=NamedRerouteUsage()))
  n86 = Power(Base=NamedRerouteUsage(), Exp=ScalarParameter() {defaultValue=5.9459, group=Quail Wink / Loan Float / Evoke Find, parameterName=AO Exponent})
  NamedRerouteDeclaration(None=Multiply(A=Multiply(A=@n81, B=@n85), B=Saturate(None=@n86))) {name=Reflection Vector, nodeColor=<rgba 1 0.9844 0 1>}
  NamedRerouteDeclaration(None=@n30.R) {name=Reflection Edge, nodeColor=<rgba 1 0 0.0937 1>}
  n87 = TextureSample!(
   UVs
   Tex
   Bias=ScalarParameter() {defaultValue=-3, group=Quail Wink / Loan Float / Evoke Find, parameterName=Virtual Coords - Mip Bias}
   Apply View MipBias
  ) {
   mipValueMode=TMVM_MipBias
   samplerSource=SSM_Wrap_WorldGroupSettings
   samplerType=SAMPLERTYPE_Masks
   texture=&T_LemonExam_Sharp_FX
  }
  NamedRerouteDeclaration(None=Saturate(None=Power(Base=@n87.A) {constExponent=10})) {name=Pattern Clip Mask, nodeColor=<rgba 0.0312 1 0 1>}
  n88 = NamedRerouteUsage()
  n89 = LinearInterpolate(A=@n88, B=Multiply(A=@n88) {constB=-1}, Alpha=NamedRerouteUsage())
  n90 = TextureSample!(UVs=Add(A=TextureCoordinate(), B=Multiply(A=@n89) {constB=0.1}), Tex, Apply View MipBias) {samplerSource=SSM_Wrap_WorldGroupSettings, samplerType=SAMPLERTYPE_Masks, texture=&T_LemonExam_Sharp_FX}
  NamedRerouteDeclaration(None=Multiply(A=Power(Base=@n90.A) {constExponent=4}) {constB=0.5}) {name=Parallaxed_Pattern, nodeColor=<rgba 0.1484 0 1 1>}
  n91 = MaterialFunctionCall(Float2=Multiply(A=NamedRerouteUsage(), B=Constant2Vector() {g=0.1, r=0.5}) {constB=0.01}) {materialFunction=&BreakOutFloat2Components}
  n92 = NamedRerouteUsage()
  n93 = ScalarParameter() {defaultValue=0.5, group=Quail Wink / Loan Float / Evoke Find, parameterName=Noise - Random Offset Multiplier}
  n94 = MaterialFunctionCall(X=@n91.R, Y=Add(A=@n91.G, B=Multiply(A=@n92, B=@n93))) {materialFunction=&MakeFloat2}
  n95 = TextureSample(UVs=Add(A=Panner(Coordinate=@n94.Result) {speedY=0.05}, B=NamedRerouteUsage()), Tex=NamedRerouteUsage()) {samplerSource=SSM_Wrap_WorldGroupSettings, samplerType=SAMPLERTYPE_LinearColor}
  NamedRerouteDeclaration(None=@n95.R) {name=Grayscale Gradient, nodeColor=<rgba 0.0547 0 1 1>}
  NamedRerouteDeclaration(None=@n46.G) {name=ATK_01_Alpha, nodeColor=<rgba 0.8516 0 1 1>}
  NamedRerouteDeclaration(None=ComponentMask(None=@n29.M) {g=true}) {name=Curvature, nodeColor=<rgba 1 0 0.5391 1>}
  n96 = Panner(Coordinate=Multiply(A=NamedRerouteUsage(), B=Constant2Vector() {g=0.1, r=0.25}) {constB=0.01}) {speedX=0.1}
  n97 = NamedRerouteUsage()
  n98 = Panner(Coordinate=Multiply(
    A=NamedRerouteUsage()
    B=ScalarParameter() {defaultValue=0.5, group=Quail Wink / Loan Float / Evoke Find, parameterName=Shining Noise - Tiling}
   ) {constB=0.01}) {speedX=0.01, speedY=0.05}
  n99 = NamedRerouteUsage()
  NamedRerouteDeclaration(None=LinearInterpolate(
    A=TextureSample(UVs=@n96, Tex=@n97) {samplerSource=SSM_Wrap_WorldGroupSettings, samplerType=SAMPLERTYPE_LinearColor}.R
    B=TextureSample(UVs=@n98, Tex=@n99) {samplerSource=SSM_Wrap_WorldGroupSettings, samplerType=SAMPLERTYPE_LinearColor}.R
    Alpha=ScalarParameter() {defaultValue=0.5, parameterName=WALI}
   )) {name=Shining Noise, nodeColor=<rgba 0.5703 1 0 1>}
  n100 = Subtract(A=Time(), B=ScalarParameter() {defaultValue=-99999, group=Quail Wink / Loan Float / Evoke Find, parameterName=APR_GameTimeFromBP})
  n101 = Add(A=ScalarParameter() {defaultValue=1, group=Quail Wink / Loan Float / Evoke Find, parameterName=APR_AnimationDuration, sortPriority=101}) {constB=0}
  NamedRerouteDeclaration(None=Saturate(None=Divide(A=@n100, B=@n101))) {name=APR_TimeNative, nodeColor=<rgba 0.2891 1 0 1>}
  NamedRerouteDeclaration(None=ComponentMask(None=@n29.M) {r=true}) {name=AO, nodeColor=<rgba 0 0.2969 1 1>}
  NamedRerouteDeclaration(None=@n30.R) {name=Luminance Mask, nodeColor=<rgba 0.1953 0 1 1>}
  n102 = Transform(None=PixelNormalWS()) {transformSourceType=TRANSFORMSOURCE_World, transformType=TRANSFORM_Local}
  n103 = Constant3Vector() {constant=<rgba 1 0 0 0>}
  n104 = DotProduct(A=@n102, B=Multiply(A=@n103) {constB=-1})
  n105 = DotProduct(A=@n102, B=@n103)
  NamedRerouteDeclaration(None=Power(Base=Max(A=@n104, B=@n105), Exp=ScalarParameter() {defaultValue=1})) {name=Side Normals, nodeColor=<rgba 1 0 0.8203 1>}
  NamedRerouteDeclaration(None=VertexInterpolator(VS=LocalPosition() {localOrigin=InstancePreSkinning}.XYZ).PS) {name=PreskinnedLoc, nodeColor=<rgba 0.4946 0.0524 1 1>}
  NamedRerouteDeclaration(None=ComponentMask(None=@n66.RGB) {r=true}) {name=Blade Cell Noise, nodeColor=<rgba 0 0.2031 1 1>}
  n106 = ScalarParameter() {defaultValue=1, group=Quail Wink / Loan Float / Evoke Find, parameterName=APR_Stage}
  n107 = LinearInterpolate(A=Subtract(A=@n106), B=@n106, Alpha=NamedRerouteUsage()) {constB=20}
  NamedRerouteDeclaration(None=Saturate(None=@n107)) {name=APR_Stages_01, nodeColor=<rgba 0.3812 1 0.0069 1>}
  NamedRerouteDeclaration(None=VertexInterpolator(VS=PreSkinnedNormal()).PS) {name=PreskinnedNormal, nodeColor=<rgba 0.4946 0.0524 1 1>}
  n108 = Divide(A=Multiply(A=NamedRerouteUsage()) {constB=-1}, B=Abs(None=Constant() {r=32})) {constB=1024}
  n109 = VectorParameter() {group=Quail Wink / Loan Float / Evoke Find, parameterName=Triplannar3DTilling, defaultValue=<rgba 1 1 1 1>}
  n110 = VectorParameter() {group=Quail Wink / Loan Float / Evoke Find, parameterName=Triplannar3DOffset, defaultValue=<rgba 0 0 0 1>}
  n111 = MaterialFunctionCall(Float3=Add(A=Multiply(A=@n108, B=@n109.RGB), B=@n110.RGB)) {materialFunction=&BreakOutFloat3Components}
  n112 = NamedRerouteUsage()
  MaterialExpressionComponentMask_9 = ComponentMask(None=@n112) {r=true}
  n113 = Custom!(x=@MaterialExpressionComponentMask_9) {
   code="return float(x < 0) ? -1 : 1;"
   description=StepReturn
   inputs [{inputName=x, input {expression=&self:MaterialExpressionComponentMask_9}}]
  }
  MaterialExpressionComponentMask_8 = ComponentMask(None=@n112) {g=true}
  n114 = Custom(x=@MaterialExpressionComponentMask_8) {
   code="return float(x < 0) ? 1 : -1;"
   description=StepReturn
   inputs [{inputName=x, input {expression=&self:MaterialExpressionComponentMask_8}}]
  }
  MaterialExpressionComponentMask_7 = ComponentMask(None=@n112) {b=true}
  n115 = Custom(x=@MaterialExpressionComponentMask_7) {
   code="return float(x < 0) ? -1 : 1;"
   description=StepReturn
   inputs [{inputName=x, input {expression=&self:MaterialExpressionComponentMask_7}}]
  }
  n116 = MaterialFunctionCall(Float3=MaterialFunctionCall(X=@n113, Y=@n114, Z=@n115) {materialFunction=&MakeFloat3}.Result) {materialFunction=&BreakOutFloat3Components}
  n117 = Abs(None=@n112)
  n118 = MaterialFunctionCall(Float3=@n117) {materialFunction=&BreakOutFloat3Components}
  n119 = Max(A=Max(A=@n118.R, B=@n118.R), B=@n118.B)
  n120 = MaterialFunctionCall(Float3=SmoothStep(Max=@n119, Value=@n117)) {materialFunction=&BreakOutFloat3Components}
  n121 = LinearInterpolate(
   A=MaterialFunctionCall(X=Multiply(A=@n111.G, B=@n116.R), Y=@n111.B) {desc=Side, materialFunction=&MakeFloat2}.Result
   B=MaterialFunctionCall(X=Multiply(A=@n111.R, B=@n116.G), Y=@n111.B) {desc=Front, materialFunction=&MakeFloat2}.Result
   Alpha=@n120.G
  )
  n122 = MaterialFunctionCall(X=Multiply(A=@n111.R, B=@n116.B), Y=@n111.G) {desc=Top, materialFunction=&MakeFloat2}
  NamedRerouteDeclaration(None=ConstantBiasScale(None=LinearInterpolate(A=@n121, B=@n122.Result, Alpha=@n120.B))) {name=HardTriplannarUV, nodeColor=<rgba 0.0312 0 1 1>}
  n123 = Multiply(A=CameraVectorWS()) {constB=-1}
  n124 = VertexNormalWS()
  n125 = MaterialFunctionCall(Float3=Subtract(
    A=Transform(None=@n123) {transformSourceType=TRANSFORMSOURCE_World, transformType=TRANSFORM_Tangent}
    B=Transform(None=@n124) {transformSourceType=TRANSFORMSOURCE_World, transformType=TRANSFORM_Tangent}
   )) {materialFunction=&BreakOutFloat3Components}
  n126 = MaterialFunctionCall(X=@n125.R, Y=@n125.R) {materialFunction=&MakeFloat2}
  n127 = ScalarParameter() {defaultValue=-0.1, parameterName=ParallaxDepth}
  n128 = Multiply(A=@n125.B) {constB=-1}
  n129 = NamedRerouteUsage()
  NamedRerouteDeclaration(None=Divide(A=Multiply(A=@n126.Result, B=@n127), B=LinearInterpolate(A=@n125.B, B=@n128, Alpha=@n129))) {name=PrallaxDepth, nodeColor=<rgba 0.5938 0 1 1>}
  n130 = TextureCoordinate()
  NamedRerouteDeclaration(None=Step(X=ComponentMask(None=@n130) {g=true}) {constX=0.664, constY=0.664}) {name=TopHalfBlade_Msk, nodeColor=<rgba 0 1 0.2734 1>}
  NamedRerouteDeclaration(None=TextureObject() {samplerType=SAMPLERTYPE_LinearColor, texture=&T_PerlinNoise_Uncompressed}) {name=NoiseTex, nodeColor=<rgba 0.6406 1 0 1>}
  n131 = Time()
  n132 = NamedRerouteUsage()
  n133 = Constant2Vector() {g=0.1, r=0.25}
  n134 = Multiply(
   A=ScalarParameter() {defaultValue=1, group=Quail Wink / Loan Float / Evoke Find, parameterName=Noise Modulation - Strength}
   B=TextureSample(UVs=Multiply(A=@n132, B=@n133) {constB=0.01}, Tex=NamedRerouteUsage()) {samplerSource=SSM_Wrap_WorldGroupSettings, samplerType=SAMPLERTYPE_LinearColor}.R
  )
  n135 = Subtract(
   A=Power(Base=@n95.R) {constExponent=0.5}
   B=Multiply(
    A=Add(A=@n131, B=@n134)
    B=ScalarParameter() {defaultValue=0.2, group=Quail Wink / Loan Float / Evoke Find, parameterName=Noise Modulation - Speed}
   )
  )
  n136 = TextureSampleParameter2D(UVs=@n135) {
   group=Quail Wink / Loan Float / Evoke Find
   parameterName=Gradient Color
   samplerSource=SSM_Wrap_WorldGroupSettings
   samplerType=SAMPLERTYPE_LinearColor
   texture=/LemonExam/Assets/Textures/CA_LemonExam_Gradient
  }
  n137 = Constant() {r=0.1}
  n138 = TextureSample(UVs=@n135) {samplerSource=SSM_Wrap_WorldGroupSettings, samplerType=SAMPLERTYPE_LinearColor, texture=/LemonExam/Assets/Textures/CA_LemonExam_Gradient_02}
  n139 = Power(Base=Multiply(A=@n138.RGB, B=Constant3Vector() {constant=<rgba 0.1543 0.4422 0.8229 1>}) {constB=0.8}) {constExponent=0.4545}
  NamedRerouteDeclaration(None=LinearInterpolate(A=Desaturation(None=@n136.RGB, Fraction=@n137), B=@n139, Alpha=NamedRerouteUsage())) {name=Gradient Map, nodeColor=<rgba 0.6172 1 0 1>}
  n140 = Constant4Vector() {constant=<rgba 1 2 0 1>}
  NamedRerouteDeclaration(None=Saturate(None=MaterialFunctionCall(Input=@n107, Input Low=@n140.R, Input High=@n140.R, Target Low=@n140.R, Target High=@n140.R) {materialFunction=&RemapValueRange}.Result)) {name=APR_Stages_02, nodeColor=<rgba 0.3812 1 0.0069 1>}
  NamedRerouteDeclaration(None=@n1.Emissive Color) {name=Emissive In, nodeColor=<rgba 1 0 0.5391 1>}
  n141 = Power(
   Base=NamedRerouteUsage()
   Exp=ScalarParameter() {defaultValue=1, group=Quail Wink / Loan Float / Evoke Find, parameterName=Shining Noise - Exp}
  )
  n142 = Saturate(None=Multiply(
    A=Saturate(None=@n141)
    B=ScalarParameter() {defaultValue=0.5, group=Quail Wink / Loan Float / Evoke Find, parameterName=Shining Noise - Strength}
   ))
  n143 = NamedRerouteUsage()
  n144 = ScalarParameter() {defaultValue=1, group=Quail Wink / Loan Float / Evoke Find, parameterName=AO Strength}
  n145 = Multiply(A=Saturate(None=Power(Base=@n143, Exp=@n144)), B=NamedRerouteUsage())
  n146 = Multiply(A=@n145, B=ScalarParameter() {defaultValue=0.6, group=Quail Wink / Loan Float / Evoke Find, parameterName=Darkening Intensity})
  n147 = Add(
   A=LocalPosition().Z
   B=ScalarParameter() {defaultValue=-17.4273, group=Quail Wink / Loan Float / Evoke Find, parameterName=Darkening Position}
  )
  n148 = ScalarParameter() {defaultValue=30, group=Quail Wink / Loan Float / Evoke Find, parameterName=Darkening Thickness}
  n149 = Add(A=@n147, B=@n148)
  n150 = Add(A=OneMinus(None=@n147), B=@n148)
  n151 = Saturate(None=Divide(A=Min(A=@n149, B=@n150), B=ScalarParameter() {defaultValue=44, group=Quail Wink / Loan Float / Evoke Find, parameterName=Darkening Fade}))
  n152 = Fresnel(ExponentIn=ScalarParameter() {defaultValue=0.7, group=Quail Wink / Loan Float / Evoke Find, parameterName=Gradient Fresnel - Exp}) {baseReflectFraction=0}
  n153 = Saturate(None=Max(
    A=OneMinus(None=@n152)
    B=ScalarParameter() {defaultValue=0.2, group=Quail Wink / Loan Float / Evoke Find, parameterName=Gradient Fresnel - Min}
   ))
  n154 = VectorParameter() {group=Quail Wink / Loan Float / Evoke Find, parameterName=Gradient Fresnel Color, defaultValue=<rgba 1 1 1 1>}
  n155 = Multiply(A=@n152, B=@n154.RGB)
  n156 = Multiply(
   A=Multiply(A=LinearInterpolate(A=@n145, B=@n146, Alpha=@n151), B=Max(A=@n153, B=@n155))
   B=ScalarParameter() {defaultValue=1, group=Quail Wink / Loan Float / Evoke Find, parameterName=Gradient Strength}
  )
  n157 = OneMinus(None=NamedRerouteUsage())
  n158 = Add(
   A=Saturate(None=MaterialFunctionCall!(Blend=@n142, Base=@n156) {materialFunction=/Engine/Functions/Engine_MaterialFunctions03/Blends/Blend_ColorDodge}.Result)
   B=Multiply(A=NamedRerouteUsage(), B=NamedRerouteDeclaration(None=@n157) {name=Curvature Inverted, nodeColor=<rgba 0.9219 1 0 1>})
  )
  n159 = ScalarParameter() {defaultValue=0.7, group=Quail Wink / Loan Float / Evoke Find, parameterName=Curvature Value}
  n160 = NamedRerouteUsage()
  NamedRerouteDeclaration(None=LinearInterpolate(A=@n158, B=Multiply(A=@n158, B=@n159), Alpha=Saturate(None=@n160))) {name=Effect, nodeColor=<rgba 0 0.8594 1 1>}
  NamedRerouteDeclaration(None=StaticComponentMaskParameter(None=@n66.RGB) {defaultB=true, group=Quail Wink / Loan Float / Evoke Find, parameterName=Blade Mask}) {name=Blade Mask, nodeColor=<rgba 1 0.8203 0 1>}
  NamedRerouteDeclaration(None=Step(
    Y=ScalarParameter() {defaultValue=1, group=Quail Wink / Loan Float / Evoke Find, parameterName=Emissive Removal}
    X=StaticComponentMaskParameter(None=@n87.RGB) {defaultR=true, group=Quail Wink / Loan Float / Evoke Find, parameterName=Pattern Mask}
   )) {name=Pattern Mask, nodeColor=<rgba 0.5 0 1 1>}
  OneMinus(None=@n11)
  Saturate(None=@n139)
  ScalarParameter() {group=Quail Wink / Loan Float / Evoke Find, parameterName=SSS - Opacity}
  ScalarParameter() {parameterName=AppearDirection}
  Step(Y=@n119, X=@n117)
  VectorParameter() {group=Quail Wink / Loan Float / Evoke Find, parameterName=Flow Noise - Speed, defaultValue=<rgba 0.05 0.1 0 0>}
  VectorParameter() {group=Quail Wink / Loan Float / Evoke Find, parameterName=Shining Noise Offset, defaultValue=<rgba 0.01 0.05 0 0>}
 ]
}
